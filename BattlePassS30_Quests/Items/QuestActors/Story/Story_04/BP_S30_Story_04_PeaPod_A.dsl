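; Pea pod quest actor for a Season 30 Battle Pass story quest that fires a one-shot explosion effect when a pawn interacts with it.
; On BeginPlay it binds to OnInteract, and the handler executes a burst gameplay cue on itself using its explosion tag.
; The construction script shifts the hidden interaction collision mesh by a fixed offset rotated to the actor's facing.
; The overlap and tick events are empty.
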(event EventBeginPlay
  (|CallParentFunction)
  (Interaction|BindEventtoOnInteract (delegate Custom|OnInteract_Event)))

(event Custom|OnInteract_Event
  (params (SelfActor :type Building%20Actor%20Object%20Reference) (InteractingPawn :type Fort%20Pawn%20Object%20Reference))
  (GameplayCue|ExecuteGameplayCueOnActor(Burst) self (Variables|Default|GetExplosionGCTag) (Ability|GameplayCue|MakeGameplayCueParameters 0.0)))

(event Collision|EventActorBeginOverlap
  (params (OtherActor :type Actor%20Object%20Reference)))

(event EventTick
  (params (DeltaSeconds :type Float%20%28single-precision%29)))

(fn ConstructionScript
  (|CallParentFunction)
  (Transformation|SetWorldLocation (Variables|Components|GetSM_HiddenInteractionCollision) (+ (Transformation|GetWorldLocation (Variables|Components|GetSM_HiddenInteractionCollision)) (Math|Vector|RotateVector (Math|Vector|MakeVector 12.0 24.0) (Transformation|GetActorRotation))) false false))
